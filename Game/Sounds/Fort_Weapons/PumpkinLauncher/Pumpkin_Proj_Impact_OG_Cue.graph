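# Sound cue for the Pumpkin Launcher projectile impact explosion, classed under ranged weapon projectile effects.
# A quality-level root sends the first two levels to an equal-weight random pick of two explosion waves, and the last two to a modulated single wave.
# It references RPG explosion attenuation, and its concurrency override, set to prevent new sounds, supersedes the assigned one-per-owner concurrency set.
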SoundCue /Game/Sounds/Fort_Weapons/PumpkinLauncher/Pumpkin_Proj_Impact_OG_Cue
root=QualityLevel_0
nodes [
 {id=Modulator_0, children [WavePlayer_0]}
 {id=QualityLevel_0, children [Random_1,Random_1,Modulator_0,Modulator_0]}
 {id=Random_1, children [WavePlayer_5,WavePlayer_6], properties {weights [1,1]}}
 {id=WavePlayer_0, properties {soundWaveAssetPtr=/Game/Sounds/Fort_Weapons/PumpkinLauncher/SoundWaves/Pumpkin_Explode_01}}
 {id=WavePlayer_5, properties {soundWaveAssetPtr=&Pumpkin_Explode_01}}
 {id=WavePlayer_6, properties {soundWaveAssetPtr=/Game/Sounds/Fort_Weapons/PumpkinLauncher/SoundWaves/Pumpkin_Explode_02}}
]
properties {
 attenuationSettings=/Game/Sounds/Attenuation/Fort_Explosion_RPG_Att
 bOverrideConcurrency=true
 maxDistance=13000
 soundClassObject=&Fort_WeaponRanged_Projectile_Effects
 volumeMultiplier=1
 attenuationOverrides {bAttenuate=false, falloffDistance=3400, attenuationShapeExtents {x=600}}
 concurrencyOverrides {resolutionRule=PreventNew}
 concurrencySet [/Game/Sounds/Concurrency_Rules/OneMaximum_Owner_StopOldest]
}
