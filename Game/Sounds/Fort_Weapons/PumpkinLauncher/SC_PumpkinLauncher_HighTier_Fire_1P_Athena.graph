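# First-person firing sound cue for the high-tier Pumpkin Launcher weapon, going by its Athena asset name.
# A mixer layers a random rocket-launcher start, a random pitched-down launch laugh at reduced level, and a boosted rocket shot delayed slightly behind them.
# Each layer passes through a modulator for variation.
# First-person weapon attenuation, sound class and gunfire concurrency settings govern playback.
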SoundCue /Game/Sounds/Fort_Weapons/PumpkinLauncher/SC_PumpkinLauncher_HighTier_Fire_1P_Athena
root=Mixer_0
nodes [
 {id=Delay_0, children [Modulator_0], properties {delayMax=0.15, delayMin=0.15}}
 {id=Mixer_0, children [Modulator_6,Modulator_3,Delay_0], properties {inputVolume [1,0.5,3]}}
 {id=Modulator_0, children [WavePlayer_3]}
 {id=Modulator_3, children [Random_0], properties {pitchMax=0.95, pitchMin=0.85}}
 {id=Modulator_6, children [Random_3]}
 {id=Random_0, children [WavePlayer_10, WavePlayer_11], properties {weights [1,1]}}
 {id=Random_3, children [WavePlayer_18, WavePlayer_19], properties {bShouldExcludeFromBranchCulling=true, weights [1,1]}}
 {id=WavePlayer_10, properties {soundWaveAssetPtr=/Game/Sounds/Fort_Weapons/PumpkinLauncher/SoundWaves/PumpkinLauncher_Prj_Launch_Laugh_01}}
 {id=WavePlayer_11, properties {soundWaveAssetPtr=/Game/Sounds/Fort_Weapons/PumpkinLauncher/SoundWaves/PumpkinLauncher_Prj_Launch_Laugh_02}}
 {id=WavePlayer_18, properties {soundWaveAssetPtr=/Game/Sounds/Fort_Weapons/Launchers/rocket_launcher_start_02}}
 {id=WavePlayer_19, properties {soundWaveAssetPtr=/Game/Sounds/Fort_Weapons/Launchers/rocket_launcher_start_01}}
 {id=WavePlayer_3, properties {soundWaveAssetPtr=/Game/Sounds/Fort_Weapons/Launchers/Rocket_Launcher_Medium_ShotC01}}
]
properties {
 attenuationSettings=&Fort_Weapon_Projectile_FirstPerson_Athena_Att
 maxDistance=2097152
 soundClassObject=&Fort_WeaponRanged_Fire_1P_PSM
 volumeMultiplier=1
 attenuationOverrides {falloffDistance=5000, attenuationShapeExtents {x=1000}}
 concurrencyOverrides {bLimitToOwner=true, maxCount=3, resolutionRule=PreventNew}
 concurrencySet [&Weapon_Gunfire_FirstPerson_Conc]
}
