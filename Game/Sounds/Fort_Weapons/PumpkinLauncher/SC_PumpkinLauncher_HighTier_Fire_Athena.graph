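# Sound cue for the high-tier Pumpkin Launcher firing in Athena, built from reused rocket launcher recordings.
# A mixer blends close, medium and distant layers, each with its own attenuation and a sound class override.
# The close layer randomly picks one of two start waves and uses a third-person fire class.
# Medium and distant layers share the distant fire class, with pitch modulation on the distant shot, and gunfire concurrency sets limit stacking.
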SoundCue /Game/Sounds/Fort_Weapons/PumpkinLauncher/SC_PumpkinLauncher_HighTier_Fire_Athena
root=Mixer_0
nodes [
 {id=Attenuation_3, children [Modulator_3], properties {attenuationSettings=&Fort_Weapon_Projectile_Close_Athena_Att}}
 {
  id=Attenuation_4
  children [WavePlayer_3]
  properties {attenuationSettings=&Fort_Weapon_Projectile_Medium_Mono_Athena_Att}
 }
 {
  id=Attenuation_5
  children [Modulator_4]
  properties {attenuationSettings=&Fort_Weapon_Projectile_Distant_Mono_Athena_Att}
 }
 {id=Mixer_0, children [SoundClass_7,SoundClass_8,SoundClass_5], properties {inputVolume [0.65,0.5,0.45]}}
 {id=Modulator_3, children [Random_0], properties {pitchMax=1, pitchMin=1}}
 {id=Modulator_4, children [WavePlayer_4], properties {pitchMax=1.1}}
 {id=Random_0, children [WavePlayer_5,WavePlayer_6], properties {bShouldExcludeFromBranchCulling=true, weights [1,1]}}
 {
  id=SoundClass_5
  children [Attenuation_5]
  properties {soundClassOverride=&Fort_WeaponRanged_Fire_Distant}
 }
 {
  id=SoundClass_7
  children [Attenuation_3]
  properties {soundClassOverride=&Fort_WeaponRanged_Fire_3P_PSM}
 }
 {
  id=SoundClass_8
  children [Attenuation_4]
  properties {soundClassOverride=&Fort_WeaponRanged_Fire_Distant}
 }
 {id=WavePlayer_3, properties {soundWaveAssetPtr=/Game/Sounds/Fort_Weapons/Launchers/Rocket_Launcher_Medium_ShotC01}}
 {id=WavePlayer_4, properties {soundWaveAssetPtr=/Game/Sounds/Fort_Weapons/Launchers/Rocket_Launcher_Distant_ShotE01}}
 {id=WavePlayer_5, properties {soundWaveAssetPtr=/Game/Sounds/Fort_Weapons/Launchers/rocket_launcher_start_02}}
 {id=WavePlayer_6, properties {soundWaveAssetPtr=/Game/Sounds/Fort_Weapons/Launchers/rocket_launcher_start_01}}
]
properties {
 maxDistance=27500
 volumeMultiplier=1.3
 attenuationOverrides {falloffDistance=5000, attenuationShapeExtents {x=1000}}
 concurrencyOverrides {bLimitToOwner=true, maxCount=3, resolutionRule=PreventNew}
 concurrencySet [&Weapon_Gunfire_Local_Conc, &Weapon_Gunfire_Global_Conc]
}
