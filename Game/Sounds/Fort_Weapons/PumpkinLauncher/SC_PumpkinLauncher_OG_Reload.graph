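# Sound cue for the original Pumpkin Launcher's reload sound.
# A mixer layers a generic launcher reload wave with a slightly quieter pumpkin-specific reload wave that a fixed delay starts just under a second later, each through its own modulator.
# It plays as weapon foley with reload attenuation, and a concurrency override allows one instance, stopping the farthest first.
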SoundCue /Game/Sounds/Fort_Weapons/PumpkinLauncher/SC_PumpkinLauncher_OG_Reload
root=Mixer_0
nodes [
 {id=Delay_1, children [WavePlayer_0], properties {delayMax=0.9, delayMin=0.9}}
 {id=Mixer_0, children [Modulator_1,Modulator_2], properties {inputVolume [1,0.7732]}}
 {id=Modulator_1, children [WavePlayer_3]}
 {id=Modulator_2, children [Delay_1]}
 {id=WavePlayer_0, properties {soundWaveAssetPtr=/Game/Sounds/Fort_Weapons/PumpkinLauncher/SoundWaves/Pumpkin_Reload}}
 {id=WavePlayer_3, properties {soundWaveAssetPtr=/Game/Sounds/Fort_Weapons/Launchers/weapons_launchers_reload_01}}
]
properties {
 attenuationSettings=&Fort_Weapon_Reload_Att
 bOverrideConcurrency=true
 maxDistance=2250
 soundClassObject=&Fort_WeaponFoley
 volumeMultiplier=0.501
 attenuationOverrides {falloffDistance=5000, attenuationShapeExtents {x=1000}}
 concurrencyOverrides {maxCount=1, resolutionRule=StopFarthestThenPreventNew}
}
