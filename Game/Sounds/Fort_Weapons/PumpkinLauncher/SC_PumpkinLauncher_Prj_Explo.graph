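# Sound cue for the Pumpkin Launcher projectile explosion, layering a cackling laugh over RPG and Quad Launcher explosion samples.
# A mixer blends four randomized layers: boosted, pitched-up laugh, close RPG, medium Quad Launcher, and distant RPG.
# Each layer has its own distance attenuation and sound class override, splitting close third-person fire from distant fire.
# Cue-level concurrency override uses PreventNew, despite also referencing a stop-oldest per-owner set.
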SoundCue /Game/Sounds/Fort_Weapons/PumpkinLauncher/SC_PumpkinLauncher_Prj_Explo
root=Mixer_4
nodes [
 {
  id=Attenuation_4
  children [Random_0]
  properties {attenuationSettings=/Game/Sounds/Attenuation/Fort_Explosion_Close_Athena_Att, attenuationOverrides {falloffDistance=1600}}
 }
 {
  id=Attenuation_5
  children [Modulator_5]
  properties {
   attenuationSettings=/Game/Sounds/Attenuation/Fort_Explosion_Medium_Athena_Att
   attenuationOverrides {falloffDistance=3000, attenuationShapeExtents {x=1000}}
  }
 }
 {
  id=Attenuation_6
  children [Modulator_6]
  properties {
   attenuationSettings=/Game/Sounds/Attenuation/Fort_Explosion_Distant_Athena_Att
   attenuationOverrides {falloffDistance=3000, attenuationShapeExtents {x=1000}}
  }
 }
 {
  id=Attenuation_7
  children [Modulator_0]
  properties {
   attenuationSettings=&Fort_Explosion_Close_Athena_Att
   attenuationOverrides {falloffDistance=3000, attenuationShapeExtents {x=1000}}
  }
 }
 {id=Mixer_4, children [SoundClass_7,SoundClass_5,SoundClass_4,SoundClass_6], properties {inputVolume [1.5,1,0.7,1]}}
 {id=Modulator_0, children [Random_1], properties {pitchMax=1.75, pitchMin=1.3}}
 {id=Modulator_4, children [Attenuation_4]}
 {id=Modulator_5, children [Random_3], properties {pitchMax=1.1, pitchMin=0.9}}
 {id=Modulator_6, children [Random_2], properties {pitchMax=1.1, pitchMin=0.9}}
 {id=Random_0, children [WavePlayer_2,WavePlayer_4,WavePlayer_7], properties {bShouldExcludeFromBranchCulling=true, weights [1,1,1]}}
 {id=Random_1, children [WavePlayer_0,WavePlayer_1,WavePlayer_3], properties {bShouldExcludeFromBranchCulling=true, weights [1,1,1]}}
 {id=Random_2, children [WavePlayer_9, WavePlayer_10, WavePlayer_11], properties {bShouldExcludeFromBranchCulling=true, weights [1,1,1]}}
 {id=Random_3, children [WavePlayer_8, WavePlayer_12, WavePlayer_13], properties {bShouldExcludeFromBranchCulling=true, weights [1,1,1]}}
 {
  id=SoundClass_4
  children [Attenuation_5]
  properties {soundClassOverride=&Fort_WeaponRanged_Fire_Distant}
 }
 {id=SoundClass_5, children [Modulator_4], properties {soundClassOverride=&Fort_WeaponRanged_Fire_3P_PSM}}
 {
  id=SoundClass_6
  children [Attenuation_6]
  properties {soundClassOverride=&Fort_WeaponRanged_Fire_Distant}
 }
 {
  id=SoundClass_7
  children [Attenuation_7]
  properties {soundClassOverride=&Fort_WeaponRanged_Fire_3P_PSM}
 }
 {id=WavePlayer_0, properties {soundWaveAssetPtr=/Game/Sounds/Fort_Weapons/PumpkinLauncher/SoundWaves/PumpkinLauncher_Prj_Explo_Laugh_01}}
 {id=WavePlayer_1, properties {soundWaveAssetPtr=/Game/Sounds/Fort_Weapons/PumpkinLauncher/SoundWaves/PumpkinLauncher_Prj_Explo_Laugh_02}}
 {id=WavePlayer_10, properties {soundWaveAssetPtr=/Game/Athena/Sounds/Weapons/Explosions/RPG_Expl_Dist_02}}
 {id=WavePlayer_11, properties {soundWaveAssetPtr=/Game/Athena/Sounds/Weapons/Explosions/RPG_Expl_Dist_03}}
 {id=WavePlayer_12, properties {soundWaveAssetPtr=/Game/Athena/Sounds/Weapons/Explosions/Quad_Launcher_Expl_Med_01}}
 {id=WavePlayer_13, properties {soundWaveAssetPtr=/Game/Athena/Sounds/Weapons/Explosions/Quad_Launcher_Expl_Med_02}}
 {id=WavePlayer_2, properties {soundWaveAssetPtr=/Game/Athena/Sounds/Weapons/Explosions/RPG_Expl_Close_01}}
 {id=WavePlayer_3, properties {soundWaveAssetPtr=/Game/Sounds/Fort_Weapons/PumpkinLauncher/SoundWaves/PumpkinLauncher_Prj_Explo_Laugh_03}}
 {id=WavePlayer_4, properties {soundWaveAssetPtr=/Game/Athena/Sounds/Weapons/Explosions/RPG_Expl_Close_02}}
 {id=WavePlayer_7, properties {soundWaveAssetPtr=/Game/Athena/Sounds/Weapons/Explosions/RPG_Expl_Close_03}}
 {id=WavePlayer_8, properties {soundWaveAssetPtr=/Game/Athena/Sounds/Weapons/Explosions/Quad_Launcher_Expl_Med_03}}
 {id=WavePlayer_9, properties {soundWaveAssetPtr=/Game/Athena/Sounds/Weapons/Explosions/RPG_Expl_Dist_01}}
]
properties {
 bOverrideConcurrency=true
 maxDistance=27500
 volumeMultiplier=1
 attenuationOverrides {bAttenuate=false, falloffDistance=3400, attenuationShapeExtents {x=600}}
 concurrencyOverrides {resolutionRule=PreventNew}
 concurrencySet [/Game/Sounds/Concurrency_Rules/OneMaximum_Owner_StopOldest]
}
