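# Sound cue for the Pumpkin Launcher projectile in flight, mixing a continuous loop with a delayed one-shot air layer.
# The loop fades in through an enveloper and passes through pitch modulation, distant rocket attenuation and a projectile-effects sound class override.
# The second branch waits briefly, then plays one of two random launch-air waves with upward pitch variation under medium projectile attenuation.
# A concurrency override caps simultaneous instances.
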SoundCue /Game/Sounds/Fort_Weapons/PumpkinLauncher/SC_PumpkinLauncher_Prj_InAir_Loop_OG
root=Mixer_0
nodes [
 {id=Attenuation_1, children [Modulator_0], properties {attenuationSettings=/Game/Sounds/Attenuation/Fort_Weapon_Rocket_Distant_Athena_Att}}
 {
  id=Attenuation_4
  children [Delay_0]
  properties {attenuationSettings=&Fort_Weapon_Projectile_Medium_Mono_Athena_Att}
 }
 {id=Delay_0, children [Modulator_3], properties {delayMax=0.3, delayMin=0.3}}
 {
  id=Enveloper_2
  children [WavePlayer_2]
  properties {pitchCurve {editorCurveData {keys [<at 0 0.9849>]}}, volumeCurve {editorCurveData {keys [<at 0 0>,<at 1 1>]}}}
 }
 {id=Mixer_0, children [SoundClass_0, Attenuation_4], properties {inputVolume [1,1]}}
 {id=Modulator_0, children [Enveloper_2], properties {pitchMax=1.1}}
 {id=Modulator_3, children [Random_3], properties {pitchMax=1.5, pitchMin=1}}
 {id=Random_3, children [WavePlayer_3,WavePlayer_4], properties {bShouldExcludeFromBranchCulling=true, weights [1,1]}}
 {
  id=SoundClass_0
  children [Attenuation_1]
  properties {soundClassOverride=&Fort_WeaponRanged_Projectile_Effects}
 }
 {
  id=WavePlayer_2
  properties {bLooping=true, soundWaveAssetPtr=/Game/Sounds/Fort_Weapons/PumpkinLauncher/SoundWaves/Pumpkin_lobber_proj_01Loop}
 }
 {id=WavePlayer_3, properties {soundWaveAssetPtr=/Game/Sounds/Fort_Weapons/PumpkinLauncher/SoundWaves/Pumpkin_Launch_Air_01}}
 {id=WavePlayer_4, properties {soundWaveAssetPtr=/Game/Sounds/Fort_Weapons/PumpkinLauncher/SoundWaves/Pumpkin_Launch_Air_02}}
]
properties {
 bDebug=true
 bOverrideConcurrency=true
 maxDistance=15000
 volumeMultiplier=0.7
 attenuationOverrides {falloffDistance=3400, attenuationShapeExtents {x=600}}
 concurrencyOverrides {maxCount=3}
}
