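# Sound cue for the close pass-by of a Pumpkin Launcher projectile, built from reused meteor "almost landed" sounds.
# A mixer layers both waves simultaneously and feeds a modulator at the root for pitch and volume variation.
# It plays much quieter than the source through the shared in-air launcher projectile attenuation and a third-person ranged weapon fire sound class.
# A concurrency override caps simultaneous instances per owner, stopping the quietest.
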SoundCue /Game/Sounds/Fort_Weapons/PumpkinLauncher/SC_PumpkinLauncher_Prj_PassBy_Close
root=Modulator_0
nodes [
 {id=Mixer_1, children [WavePlayer_2,WavePlayer_5], properties {inputVolume [1,1]}}
 {id=Modulator_0, children [Mixer_1]}
 {id=WavePlayer_2, properties {soundWaveAssetPtr=/Game/Sounds/Fort_GamePlay_Sounds/Meteor/Meteor_Almost_Landed_01}}
 {id=WavePlayer_5, properties {soundWaveAssetPtr=/Game/Sounds/Fort_GamePlay_Sounds/Meteor/Meteor_Almost_Landed_02}}
]
properties {
 attenuationSettings=/Game/Sounds/Attenuation/Fort_Weapon_Launcher_Proj_InAir_Default_Att
 bOverrideConcurrency=true
 maxDistance=12750
 soundClassObject=&Fort_WeaponRanged_Fire_3P_PSM
 volumeMultiplier=0.1
 attenuationOverrides {bAttenuate=false, falloffDistance=5000, stereoSpread=1500, attenuationShapeExtents {x=1000}}
 concurrencyOverrides {bLimitToOwner=true, maxCount=3, resolutionRule=StopQuietest}
}
